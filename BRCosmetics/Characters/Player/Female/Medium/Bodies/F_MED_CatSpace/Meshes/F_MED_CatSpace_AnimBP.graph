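# Post-copy AnimBP for the F_MED_CatSpace female body cosmetic, correcting deformation and hood placement.
# The attached parent's pose feeds a deform Control Rig, a weighted neck/spine/head hood constraint, then two additive hood bone offsets.
# Event graph flags for moving jumps or parachuting and crouched turning in place drive the rig; the former also enables the constraint.
# Head-to-spine_05 twist sets both offset alphas, one limited to crouched aiming.
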AnimBlueprint /BRCosmetics/Characters/Player/Female/Medium/Bodies/F_MED_CatSpace/Meshes/F_MED_CatSpace_AnimBP
target_skeleton=&/F_MED_CatSpace_Skeleton
graphs [
 AnimGraph {nodes [
   n1 = Control Rig!(
    Source=Copy Pose From Mesh() {bUseAttachedParent=true}
    Alpha
    static_crouching_rotation:bool=Get static_crouching_rotation() -> (static_crouching_rotation:bool)
    move_jump:bool=Get move_jump() -> (move_jump:bool)
   ) -> (Pose) {
    alphaBoolBlend {blendOption=Linear}
    controlRigAssetReference {blueprintRigClass=&/F_MED_CatSpace_DeformRig.F_MED_CatSpace_DeformRig_C}
   }
   n2 = "Constraint - hood (PPPPP)"!(
    ComponentPose=Local To Component(LocalPose=@n1)
    bAlphaBoolEnabled:bool=Get move_jump()
    ConstraintWeights_0=0.285
    ConstraintWeights_1:float=0.2788
    ConstraintWeights_2:float=0.2121
    ConstraintWeights_3:float=0.1745
    ConstraintWeights_4:float=0.0496
   ) -> (Pose) {
    alphaInputType=Bool
    alphaBoolBlend {blendInTime=0.5, blendOption=Linear, blendOutTime=0.5}
    boneToModify {boneName=hood}
    constraintSetup [
     {targetBone {boneName=neck_02}}
     {targetBone {boneName=spine_05}}
     {targetBone {boneName=head}}
     {targetBone {boneName=neck_01}}
     {targetBone {boneName=spine_04}}
    ]
    constraintWeights [1,1,1,1,1]
   }
   n3 = "Transform (Modify) Bone - Bone: hood"(ComponentPose=@n2, Alpha=Get hood() -> (hood:double), Translation="0,-5,0", Rotation="0,-50,0", Scale="1,1,1") {
    rotationMode=BMM_Additive
    rotationSpace=BCS_BoneSpace
    translationMode=BMM_Additive
    translationSpace=BCS_BoneSpace
    alphaBoolBlend {blendInTime=0.1, blendOption=Linear, blendOutTime=0.2}
    alphaScaleBiasClamp {bInterpResult=true, interpSpeedDecreasing=20, interpSpeedIncreasing=20}
    boneToModify {boneName=hood}
    rotation {yaw=-50}
    translation {y=-5}
   }
   n4 = "Transform (Modify) Bone - Bone: hood"(ComponentPose=@n3, Alpha=Get head_up() -> (head_up:double), Translation="0,0,0", Rotation="20,0,0", Scale="1,1,1") {
    rotationMode=BMM_Additive
    rotationSpace=BCS_BoneSpace
    translationMode=BMM_Additive
    translationSpace=BCS_BoneSpace
    alphaBoolBlend {blendInTime=0.1, blendOption=Linear, blendOutTime=0.2}
    alphaScaleBiasClamp {bInterpResult=true, interpSpeedDecreasing=1, interpSpeedIncreasing=1}
    boneToModify {boneName=hood}
    rotation {pitch=20}
   }
   Output Pose(Result=Component To Local(ComponentPose=@n4)) {layerGroup=None}
  ]}
 EventGraph {nodes [
   Event BlueprintUpdateAnimation()
   K2Node_Composite_0 = Hood_CollapseGraph!() -> (AlphaOutput:float=~) {is=title K2Node_Composite}
   > Set hood(hood:double=@K2Node_Composite_0) -> (Output_Get:double)
   K2Node_Composite_1 = static_crouching_rotation_CollapseGraph!() -> (ReturnValue:bool=~) {is=title K2Node_Composite}
   > Set static_crouching_rotation(static_crouching_rotation:bool=@K2Node_Composite_1) -> (Output_Get:bool)
   n1 = Get bIsMoving2D()
   n2 = Get bIsJumping()
   > Set move_jump(move_jump:bool=OR Boolean(A=AND Boolean(A=@n2, B=@n1), B=Get bIsParachuteOpened())) -> (Output_Get:bool)
   n3 = GetOwnerSkeletalMeshComponent()
   n4 = AND Boolean(A=Get bIsCrouching(), B=Get bIsTargeting())
   > Set head_up(head_up:double=float * float(
     A="To Float (Boolean)"(InBool=@n4)
     B=GetAlphaFromReferenceTransformDelta(SkeletalMeshComponent=@n3, BoneName=head, BaseName=spine_05, InputRange=(X=-10,Y=-40), TwistAxis="1,0,0")
    ) {is=title K2Node_PromotableOperator}) -> (Output_Get:double)
  ]}
 Hood_CollapseGraph {
  path=EventGraph.K2Node_Composite_0.Hood_CollapseGraph
  nodes [
   Inputs!()
   Outputs!(AlphaOutput:float=GetAlphaFromReferenceTransformDelta(SkeletalMeshComponent=GetOwnerSkeletalMeshComponent(), BoneName=head, BaseName=spine_05, InputRange=(X=-10,Y=-56), TwistAxis="1,0,0"))
  ]
 }
 static_crouching_rotation_CollapseGraph {
  path=EventGraph.K2Node_Composite_1.static_crouching_rotation_CollapseGraph
  nodes [
   Inputs()
   n1 = Get bIsMoving2D()
   n2 = "Absolute (Float)"(A=BreakRotator(InRot=Get PawnDeltaRotation() -> (PawnDeltaRotation:Rotator)).Yaw)
   Outputs!(ReturnValue:bool=AND Boolean!(A="float > float"(A=@n2, B=2), B=Get bIsCrouching(), C:bool=NOT Boolean(A=@n1)) -> (ReturnValue))
  ]
 }
 NewFunction {nodes [
   NewFunction() {is=title K2Node_FunctionEntry}
   > Return Node!(
    execute
    AlphaOutput:float=GetAlphaFromReferenceTransformDelta(SkeletalMeshComponent=GetOwnerSkeletalMeshComponent(), BoneName=head, BaseName=spine_05, InputRange=(X=-10,Y=-56), TwistAxis="1,0,0")
   )
  ]}
]
variables [
 head_up {type=number}
 hood {type=number}
 move_jump {type=boolean}
 static_crouching_rotation {type=boolean}
]
