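# Control Rig for the face accessory part of the F_MED_CatSpace player body, over a spine-to-head skeleton with left and right side and back dyn hair chains.
# The forward solve holds a single Set Relative Transform on dyn_back_hair_BASE_r, using that bone as its own parent with an identity value.
# It propagates to children but appears to be a placeholder, as no other bone is driven and no hair simulation is present.
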ControlRigBlueprint /BRCosmetics/Characters/Player/Female/Medium/Bodies/F_MED_CatSpace/Meshes/Parts/F_MED_CatSpace_FaceAcc_CtrlRig
graphs [
 RigVMFunctionLibrary {links [], nodes []}
 RigVMModel {nodes [
   RigUnit_BeginExecution = &RigUnit_BeginExecution()
   RigUnit_SetRelativeTransformForItem = /Script/ControlRig.RigUnit_SetRelativeTransformForItem(
    Child=(Type=Bone,Name="dyn_back_hair_BASE_r")
    Parent=(Type=Bone,Name="dyn_back_hair_BASE_r")
    Value=(Rotation=(X=0,Y=0,Z=0,W=1),Translation=(X=0,Y=0,Z=0),Scale3D=(X=1,Y=1,Z=1))
    Weight=1
    bParentInitial=False
    bPropagateToChildren=True
    ExecutePin=@RigUnit_BeginExecution.ExecutePin
   ) {
    is=struct RigVMUnitNode
    template="Set Relative Transform::Execute(in Child,in Parent,in bParentInitial,in Value,in Weight,in bPropagateToChildren)"
   }
  ]}
]
hierarchy [
 root {loc=<0 0 0>, rot=<0 0 0 1>}
 pelvis {parent=root, loc=<0 -0.5301 98.0571>, rot=<0.0127 0.707 -0.0127 -0.707>}
 spine_01 {parent=pelvis, loc=<6.1208 -0.9328 0>, rot=<0 0 0.034 -0.9994>}
 spine_02 {parent=spine_01, loc=<6.5996 -0.2095 0>, rot=<0 0 0.0238 -0.9997>}
 spine_03 {parent=spine_02, loc=<6.4914 -0.1237 0>, rot=<0 0 0.0149 -0.9999>}
 spine_04 {parent=spine_03, loc=<6.9598 0.0794 0>, rot=<0 0 -0.0132 -0.9999>}
 spine_05 {parent=spine_04, loc=<6.9987 0 0>, rot=<0 0 -0.0015 -1>}
 neck_01 {parent=spine_05, loc=<11.3138 0.1529 0>, rot=<0 0 -0.2003 -0.9797>}
 neck_02 {parent=neck_01, loc=<4.5627 0 0>, rot=<0 0 0.0504 -0.9987>}
 head {parent=neck_02, loc=<5.3914 0 0>, rot=<0 0 0.115 -0.9934>}
 dyn_side_hair_BASE_l {parent=head, loc=<-0.5146 0.7208 -7.3747>, rot=<-0.0131 -0.9991 0.0009 0.04>}
 dyn_side_hair_01_l {parent=dyn_side_hair_BASE_l, loc=<0 0 0>, rot=<0 0 0 1>}
 dyn_side_hair_02_l {parent=dyn_side_hair_01_l, loc=<2.5464 0 0>, rot=<0 0 0 1>}
 dyn_back_hair_BASE_l {parent=head, loc=<7.2819 -5.4001 -4.4084>, rot=<0.0302 -0.9035 0.4274 -0.0106>}
 dyn_back_hair_01_l {parent=dyn_back_hair_BASE_l, loc=<0 0 0>, rot=<0 0 0 1>}
 dyn_back_hair_02_l {parent=dyn_back_hair_01_l, loc=<5.5544 0 0>, rot=<0 0 0 1>}
 dyn_back_hair_BASE_r {parent=head, loc=<7.4473 -5.8221 3.5537>, rot=<-0.001 -0.486 0.8736 -0.0246>}
 dyn_back_hair_01_r {parent=dyn_back_hair_BASE_r, loc=<0 0 0>, rot=<0 0 0 1>}
 dyn_back_hair_02_r {parent=dyn_back_hair_01_r, loc=<5.7354 0 0>, rot=<0 0 0 1>}
 dyn_side_hair_BASE_r {parent=head, loc=<1.6352 6.0336 6.4617>, rot=<-0.0018 -0.0053 0.9993 -0.0383>}
 dyn_side_hair_01_r {parent=dyn_side_hair_BASE_r, loc=<0 0 0>, rot=<0 0 0 1>}
 dyn_side_hair_02_r {parent=dyn_side_hair_01_r, loc=<5.6269 0 0.1299>, rot=<0 -0.1045 0 -0.9945>}
 hair_block {parent=spine_05, loc=<0 0 0>, rot=<0 0 0 1>}
 root {parent=root, type=Socket, loc=<0 0 0>, rot=<0 0 0 1>}
]
