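# Physics asset for the F_MED_CatSpace face accessory mesh, defining bodies and joints for dynamic hair on a female player cosmetic.
# Four hair chains (back and side, left and right) each run from a tiny base capsule on the head through two capsule segments with narrow swing and very tight twist limits.
# Head, upper-spine and hair_block capsules add body volumes, with free-motion constraints linking head and hair_block to spine_05 and each chain base to the head.
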PhysicsAsset /BRCosmetics/Characters/Player/Female/Medium/Bodies/F_MED_CatSpace/Meshes/Parts/F_MED_CatSpace_FaceAcc_Physics
bodies {
 dyn_back_hair_01_l [{
   shapeType=Capsule
   length=5.5625
   radius=2.6797
   shapeName=dyn_hair_02_capsule
   center=<2.9648 -0.3612 1.0204>
   rotation=<rot -90 176.4237 -88.5779>
  }]
 dyn_back_hair_01_r [{
   shapeType=Capsule
   length=4.4706
   radius=2.2044
   shapeName=dyn_hair_back_02_r_capsule
   center=<3.7414 -0.1745 0.9083>
   rotation=<rot -85.0002 0.0002 88.3037>
  }]
 dyn_back_hair_02_l [{
   shapeType=Capsule
   length=4.6508
   radius=2.5679
   shapeName=dyn_hair_03_capsule
   center=<2.6506 -0.0209 0.8202>
   rotation=<rot 90 -165.9638 -74.426>
  }]
 dyn_back_hair_02_r [{
   shapeType=Capsule
   length=4.6875
   radius=2
   shapeName=dyn_hair_back_03_r_capsule
   center=<2.9305 -0.1886 0.5339>
   rotation=<rot 85.0001 -179.9997 92.1864>
  }]
 dyn_back_hair_BASE_l [{shapeType=Capsule, length=0.505, radius=0.505, shapeName=dyn_back_hair_BASE_l_capsule, rotation=<rot 0 -90 0>}]
 dyn_back_hair_BASE_r [{shapeType=Capsule, length=0.505, radius=0.505, shapeName=dyn_back_hair_BASE_r_capsule, rotation=<rot 0 -90 0>}]
 dyn_side_hair_01_l [{
   shapeType=Capsule
   length=5.9667
   radius=1.3438
   shapeName=dyn_hair_01_l_capsule
   center=<2.4661 0.2256 0.1176>
   rotation=<rot -80.6557 -106.7474 0>
  }]
 dyn_side_hair_01_r [{
   shapeType=Capsule
   length=1.1875
   radius=2.1562
   shapeName=dyn_hair_02_r_capsule
   center=<2.7274 0.1629 -0.0217>
   rotation=<rot 90 -21.1366 -20.3111>
  }]
 dyn_side_hair_02_l [{
   shapeType=Capsule
   length=7.3944
   radius=1.5938
   shapeName=dyn_hair_02_l_capsule
   center=<2.8851 0.1704 -0.2046>
   rotation=<rot -82.1864 -108.6474 180>
  }]
 dyn_side_hair_02_r [{
   shapeType=Capsule
   length=4.8956
   radius=1.9169
   shapeName=dyn_hair_03_r_capsule
   center=<3.4576 -1.2718 0.2929>
   rotation=<rot 84.3811 -4.0098 -0.0003>
  }]
 dyn_side_hair_BASE_l [{shapeType=Capsule, length=0.505, radius=0.505, shapeName=dyn_side_hair_BASE_l_capsule, rotation=<rot 0 -90 0>}]
 dyn_side_hair_BASE_r [{shapeType=Capsule, length=0.505, radius=0.505, shapeName=dyn_side_hair_BASE_r_capsule, rotation=<rot 0 -90 0>}]
 hair_block [{shapeType=Capsule, length=15, radius=8, shapeName=hair_block_capsule, center=<6.713 -5.6215 -0.1718>, rotation=<rot 0 -0.0003 20>}]
 head [
  {shapeType=Capsule, length=6.7534, radius=6.3369, center=<3.1734 2.6019 0.0541>, rotation=<rot 0.5016 -90.415 -88.3249>}
  {
   shapeType=Capsule
   length=3.3889
   radius=6.0811
   shapeName=head_capsule
   center=<1.6133 -0.0375 -0.0239>
   rotation=<rot -89.1519 179.9981 -180>
  }
 ]
 spine_05 [
  {shapeType=Capsule, length=8.0453, radius=4.3542, shapeName=spine_05_capsule, center=<8.3727 -1.9414 0>}
  {
   shapeType=Capsule
   length=4.3766
   radius=10.8207
   shapeName=spine_05_capsule
   center=<0.9477 1.8096 -0.0043>
   rotation=<rot -89.5488 -69.2055 73.8815>
  }
 ]
}
constraints [
 {
  bone1Name=head
  bone2Name=spine_05
  swing1LimitDegrees=45
  swing1Motion=Free
  swing2LimitDegrees=45
  swing2Motion=Free
  twistLimitDegrees=45
  twistMotion=Free
 }
 {
  bone1Name=dyn_back_hair_BASE_l
  bone2Name=head
  swing1LimitDegrees=45
  swing1Motion=Free
  swing2LimitDegrees=45
  swing2Motion=Free
  twistLimitDegrees=45
  twistMotion=Free
 }
 {bone1Name=dyn_back_hair_01_l, bone2Name=dyn_back_hair_BASE_l, swing1LimitDegrees=10, swing2LimitDegrees=20, twistLimitDegrees=1}
 {bone1Name=dyn_back_hair_02_l, bone2Name=dyn_back_hair_01_l, swing1LimitDegrees=10, swing2LimitDegrees=20, twistLimitDegrees=1}
 {
  bone1Name=dyn_back_hair_BASE_r
  bone2Name=head
  swing1LimitDegrees=45
  swing1Motion=Free
  swing2LimitDegrees=45
  swing2Motion=Free
  twistLimitDegrees=45
  twistMotion=Free
 }
 {bone1Name=dyn_back_hair_01_r, bone2Name=dyn_back_hair_BASE_r, swing1LimitDegrees=10, swing2LimitDegrees=10, twistLimitDegrees=1}
 {bone1Name=dyn_back_hair_02_r, bone2Name=dyn_back_hair_01_r, swing1LimitDegrees=10, swing2LimitDegrees=10, twistLimitDegrees=1}
 {
  bone1Name=dyn_side_hair_BASE_l
  bone2Name=head
  swing1LimitDegrees=45
  swing1Motion=Free
  swing2LimitDegrees=45
  swing2Motion=Free
  twistLimitDegrees=45
  twistMotion=Free
 }
 {bone1Name=dyn_side_hair_01_l, bone2Name=dyn_side_hair_BASE_l, swing1LimitDegrees=15, swing2LimitDegrees=15, twistLimitDegrees=1}
 {bone1Name=dyn_side_hair_02_l, bone2Name=dyn_side_hair_01_l, swing1LimitDegrees=10, swing2LimitDegrees=15, twistLimitDegrees=1}
 {
  bone1Name=dyn_side_hair_BASE_r
  bone2Name=head
  swing1LimitDegrees=45
  swing1Motion=Free
  swing2LimitDegrees=45
  swing2Motion=Free
  twistLimitDegrees=45
  twistMotion=Free
 }
 {bone1Name=dyn_side_hair_01_r, bone2Name=dyn_side_hair_BASE_r, swing1LimitDegrees=15, swing2LimitDegrees=15, twistLimitDegrees=1}
 {bone1Name=dyn_side_hair_02_r, bone2Name=dyn_side_hair_01_r, swing1LimitDegrees=10, swing2LimitDegrees=15, twistLimitDegrees=5}
 {
  bone1Name=hair_block
  bone2Name=spine_05
  swing1LimitDegrees=45
  swing1Motion=Free
  swing2LimitDegrees=45
  swing2Motion=Free
  twistLimitDegrees=45
  twistMotion=Free
 }
]
