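# Post-copy AnimBP for the CatSpace female face accessory part, following the body pose with lower-face bones layered from the pawn's head mesh.
# A hair_block bone offset lowers by default and lifts back with head rotation during crouched, non-pistol aiming, feeding head-space RigidBody physics and a Control Rig.
# The glasses bone is then partially constrained to nose bridge and cheeks, with a small nudge.
# A bracelet Flap node sits unconnected.
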AnimBlueprint /BRCosmetics/Characters/Player/Female/Medium/Bodies/F_MED_CatSpace/Meshes/Parts/F_MED_CatSpace_FaceAcc_AnimBP
target_skeleton=&/F_MED_CatSpace_FaceAcc_Skeleton
graphs [
 AnimGraph {nodes [
   n1 = Layered blend per bone(
    BasePose=Copy Pose From Mesh() {bCopyCurves=true, bUseAttachedParent=true}
    BlendPoses_0=Copy Pose From Mesh(SourceMeshComponent=Get head_mesh() -> (head_mesh:Skeletal Mesh Component Object Reference)) {bCopyCurves=true, bUseAttachedParent=true}
   ) {layerSetup [{branchFilters [
       {boneName=FACIAL_C_Jaw}
       {boneName=FACIAL_L_CheekLower}
       {boneName=FACIAL_L_CheekOuter}
       {boneName=FACIAL_R_CheekLower}
       {boneName=FACIAL_R_CheekOuter}
       {boneName=FACIAL_L_NasolabialBulge}
       {boneName=FACIAL_R_NasolabialBulge}
       {boneName=FACIAL_L_LipCorner}
       {boneName=FACIAL_R_LipCorner}
       {boneName=FACIAL_L_Masseter}
       {boneName=FACIAL_R_Masseter}
       {boneName=FACIAL_C_Chin}
       {boneName=FACIAL_L_ChinSide}
       {boneName=FACIAL_R_ChinSide}
       {boneName=FACIAL_C_UnderChin}
       {boneName=FACIAL_L_UnderChin}
       {boneName=FACIAL_R_UnderChin}
       {boneName=FACIAL_C_LowerLipRotation}
       {boneName=FACIAL_C_LipLower}
       {boneName=FACIAL_L_LipLower}
       {boneName=FACIAL_L_LipLowerOuter}
       {boneName=FACIAL_R_LipLower}
       {boneName=FACIAL_R_LipLowerOuter}
       {boneName=FACIAL_L_LipCorner}
       {boneName=FACIAL_R_LipCorner}
       {boneName=FACIAL_C_Jawline}
       {boneName=FACIAL_L_Jawline}
       {boneName=FACIAL_R_Jawline}
       {boneName=FACIAL_L_Ear}
       {boneName=FACIAL_R_Ear}
      ]}]}
   n2 = "Transform (Modify) Bone - Bone: hair_block"(ComponentPose=Local To Component(LocalPose=@n1), Alpha=1, Translation="0,0,-5", Rotation="0,0,0", Scale="1,1,1") {translationMode=BMM_Additive, alphaBoolBlend {blendOption=Linear}, boneToModify {boneName=hair_block}, translation {z=-5}}
   n3 = "Transform (Modify) Bone - Bone: hair_block"(ComponentPose=@n2, Alpha=Get head_up_FA() -> (head_up_FA:double), Translation="0,0,5", Rotation="0,0,0", Scale="1,1,1") {translationMode=BMM_Additive, alphaBoolBlend {blendOption=Linear}, boneToModify {boneName=hair_block}, translation {z=5}}
   n4 = RigidBody!(ComponentPose=@n3, Alpha, OverrideWorldGravity:Vector="0,0,-1500", ExternalForce) -> (Pose) {
    bOverrideWorldGravity=true
    overridePhysicsAsset=&/F_MED_CatSpace_FaceAcc_Physics
    simulationSpace=BaseBoneSpace
    alphaBoolBlend {blendOption=Linear}
    baseBoneRef {boneName=head}
    componentAppliedLinearAccClamp=<2000 2000 2000>
    componentLinearAccScale=<0.4 0.4 0.4>
    componentLinearVelScale=<0.4 0.4 0.4>
    overrideWorldGravity {z=-1500}
    simSpaceSettings {maxAngularAcceleration=3000, maxAngularVelocity=3000, maxLinearAcceleration=3000, maxLinearVelocity=3000}
   }
   n5 = Control Rig(Source=Component To Local(ComponentPose=@n4)) {
    alphaBoolBlend {blendOption=Linear}
    controlRigAssetReference {blueprintRigClass=&/F_MED_CatSpace_FaceAcc_CtrlRig.F_MED_CatSpace_FaceAcc_CtrlRig_C}
   }
   n6 = "Constraint - dyn_glasses (TTT)"!(
    ComponentPose=Local To Component(LocalPose=@n5)
    Alpha=0.5
    ConstraintWeights_0=0.6
    ConstraintWeights_1:float=0.2
    ConstraintWeights_2:float=0.2
   ) -> (Pose) {
    alphaBoolBlend {blendOption=Linear}
    boneToModify {boneName=dyn_glasses}
    constraintSetup [
     {transformType=Translation, targetBone {boneName=FACIAL_C_NoseBridge}}
     {transformType=Translation, targetBone {boneName=FACIAL_R_CheekInner}}
     {transformType=Translation, targetBone {boneName=FACIAL_L_CheekInner}}
    ]
    constraintWeights [1,1,1]
   }
   n7 = "Transform (Modify) Bone - Bone: dyn_glasses"!(ComponentPose=@n6, Alpha=1, Translation="0.2,0,0") -> (Pose) {
    translationMode=BMM_Additive
    translationSpace=BCS_BoneSpace
    alphaBoolBlend {blendOption=Linear}
    boneToModify {boneName=dyn_glasses}
    translation {x=0.2}
   }
   Output Pose(Result=Component To Local(ComponentPose=@n7)) {layerGroup=None}
   Flap Node - Bone: dyn_bracelet_bottom_2_base!(ComponentPose:Component Space Pose Link Structure, Alpha:float=0.5) -> (Pose:Component Space Pose Link Structure) {
    is=title FortAnimGraphNode_Flap
    bUseClamp=true
    targetClampMin=-3.14
    targetComponent=RotationZ
    alphaBoolBlend {blendOption=Linear}
    sourceDrives [{
      bUseRange=false
      drivingCurve=null
      multiplier=-0.75
      rangeMax=0
      rangeMin=0
      remappedMax=0
      remappedMin=0
      sourceComponent=RotationY
      useQuaternion=true
      sourceBone {boneName=head}
     }]
    targetBone {boneName=dyn_bracelet_bottom_2_base}
   }
  ]}
 EventGraph {nodes [
   Event BlueprintUpdateAnimation()
   K2Node_Composite_0 = head_up_FA_CollapseGraph!() -> (ReturnValue:double=~) {is=title K2Node_Composite}
   > Set head_up_FA(head_up_FA:double=@K2Node_Composite_0) -> (Output_Get:double)
   > n1 = Cast To PlayerPawn_Athena(Object=TryGetPawnOwner())
   > Set head_mesh(head_mesh:Skeletal Mesh Component Object Reference=GetSkeletalMeshForPartType(self=@n1, PartType=Head)) -> (Output_Get:Skeletal Mesh Component Object Reference)
  ]}
 head_up_FA_CollapseGraph {
  path=EventGraph.K2Node_Composite_0.head_up_FA_CollapseGraph
  nodes [
   Inputs!()
   n1 = GetOwnerSkeletalMeshComponent()
   n2 = GetWeaponCoreAnim!(self:Fort Player Animation Interface Object Reference=GetOwnerAnimationInterface!(self:Custom Character Part Anim Instance Object Reference) -> (ReturnValue:Fort Player Animation Interface Object Reference)) -> (ReturnValue:EFortWeaponCoreAnimation Enum)
   n3 = AND Boolean!(
    A=Get bIsCrouching()
    B=Get bIsTargeting()
    C:bool=NOT Boolean(A="Equal (Enum)"(A:EFortWeaponCoreAnimation Enum=@n2, B:EFortWeaponCoreAnimation Enum=Pistol))
   ) -> (ReturnValue)
   Outputs!(ReturnValue:double=float * float(
     A="To Float (Boolean)"(InBool=@n3)
     B=GetAlphaFromReferenceTransformDelta(SkeletalMeshComponent=@n1, BoneName=head, BaseName=spine_05, InputRange=(X=-10,Y=-40), TwistAxis="1,0,0")
    ) {is=title K2Node_PromotableOperator})
  ]
 }
]
variables [
 aim_crouch {type=boolean}
 head_mesh {type=SkeletalMeshComponent}
 head_up_FA {type=number}
]
